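; Focus point actor for shared events that drives a forced camera focus on players.
; On authority, it applies the forced-focus GameplayEffect to every player pawn.
; Force Focus passes enable state, duration, FOV, interp speed, offset and focus actor (optionally itself) to the ability as a stateful event on the global event router.
; FF_Test events duplicate the setup and broadcast hardcoded start and stop events targeting this actor.
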(event EventBeginPlay)

(event EventEndPlay
  (params (EndPlayReason :type EEndPlayReason%20Enum :default "Destroyed")))

(event Custom|Add%20Focus%20Ability%20to%20All%20Players
  ; Gameplay Ability To Focus
  (Networking|SwitchHasAuthority
    (:Authority
      (bind _outfortplayerpawns (Fortnite|GetFortPlayerPawns))
      (for _array_element _outfortplayerpawns
        (GameplayEffects|ApplyGameplayEffectToTarget :self (Ability|GetFortAbilitySystemComponent _array_element) :GameplayEffectClass "/Events_Shared/Gameplay/ForceFocus/GE_EventsShared_Focus_Forced.GE_EventsShared_Focus_Forced_C" :Target (Ability|GetFortAbilitySystemComponent _array_element) :Level 0.0) ; force focus GE
        (Fortnite|FortShippingLog 0 (Utilities|String|Append "Force Focus ability given to " (Fortnite|GetPlayerNameSafe _array_element) (Utilities|String|SelectString " (server)" " (client)" (Networking|IsServer))))))))

(event Custom|FF_TestSetup
  ; Gameplay Ability To Focus
  (Networking|SwitchHasAuthority
    (:Authority
      (bind _outfortplayerpawns (Fortnite|GetFortPlayerPawns))
      (for _array_element _outfortplayerpawns
        (GameplayEffects|ApplyGameplayEffectToTarget :self (Ability|GetFortAbilitySystemComponent _array_element) :GameplayEffectClass "/Events_Shared/Gameplay/ForceFocus/GE_EventsShared_Focus_Forced.GE_EventsShared_Focus_Forced_C" :Target (Ability|GetFortAbilitySystemComponent _array_element) :Level 0.0) ; force focus GE
        (Fortnite|FortShippingLog 0 (Utilities|String|Append "Force Focus ability given to " (Fortnite|GetPlayerNameSafe _array_element) (Utilities|String|SelectString " (server)" " (client)" (Networking|IsServer))))))))

(event Custom|Force%20Focus
  (params (Force%20Focus%20Enabled_ForceFocusEnabled_6_8B220E3A4C0AAFA126BD4286F5599F22 :type Boolean) (Force%20Focus%20Enabled_Duration_10_3674387B40D646512D5DCA9727EC0275 :type Float%20%28double-precision%29) (Force%20Focus%20Enabled_FocalPointFOV_13_8A2CBF5C4C7D57CB2D426DBDDF08BF60 :type Float%20%28double-precision%29) (Force%20Focus%20Enabled_InterpSpeed_15_4A87062643DA51A3E2CF258698FD5E2A :type Float%20%28double-precision%29) (Force%20Focus%20Enabled_FocalPointOffset_18_117ACAEF4A558F4CAE3F0B953A32672C :type Vector) (Force%20Focus%20Enabled_FocusActor_9_1763734A4F76A6EB17B0F985CE9A0797 :type Actor%20Object%20Reference) (Use%20Self%20as%20Focus%20Actor :type Boolean))
  ; Pass Focus params to Gameplay Ability
  (Fortnite|FortShippingLog 0 (Utilities|String|Append "ForceFocus event called with duration=" (Utilities|String|ToString(Float) Force%20Focus%20Enabled_Duration_10_3674387B40D646512D5DCA9727EC0275) ", interp speed=" (Utilities|String|ToString(Float) Force%20Focus%20Enabled_InterpSpeed_15_4A87062643DA51A3E2CF258698FD5E2A) (Utilities|String|SelectString " (server)" " (client)" (Networking|IsServer))))
  (GameplayEventRouter|BroadcastStatefulEvent (GameplayEventRouter|GetGlobalEventRouter) (Utilities|Struct|MakeEventsSharedForceFocusEvent Force%20Focus%20Enabled_ForceFocusEnabled_6_8B220E3A4C0AAFA126BD4286F5599F22 Force%20Focus%20Enabled_Duration_10_3674387B40D646512D5DCA9727EC0275 Force%20Focus%20Enabled_FocalPointFOV_13_8A2CBF5C4C7D57CB2D426DBDDF08BF60 Force%20Focus%20Enabled_InterpSpeed_15_4A87062643DA51A3E2CF258698FD5E2A Force%20Focus%20Enabled_FocalPointOffset_18_117ACAEF4A558F4CAE3F0B953A32672C (select Use%20Self%20as%20Focus%20Actor self Force%20Focus%20Enabled_FocusActor_9_1763734A4F76A6EB17B0F985CE9A0797)) 0))

(event Custom|FF_TestStart
  ; Pass Focus params to Gameplay Ability
  (GameplayEventRouter|BroadcastStatefulEvent (GameplayEventRouter|GetGlobalEventRouter) (Utilities|Struct|MakeEventsSharedForceFocusEvent true 5.0 9e+01 2.0 "0.000000,0.000000,0.000000" self) 0))

(event Custom|FF_TestStop
  ; Pass Focus params to Gameplay Ability
  (GameplayEventRouter|BroadcastStatefulEvent (GameplayEventRouter|GetGlobalEventRouter) (Utilities|Struct|MakeEventsSharedForceFocusEvent false 5.0 9e+01 1.0 "0.000000,0.000000,0.000000" self) 0))

(fn ConstructionScript)
