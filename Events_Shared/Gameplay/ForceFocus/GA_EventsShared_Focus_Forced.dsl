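; Gameplay ability for shared live events that forces the game's focal point onto an event-specified actor.
; Once playlist data is ready, it listens to a stateful ForceFocus event whose payload starts or stops focusing.
; Where cosmetic events cannot execute, it sets or clears the GameState focal point and toggles the special event mutator's lock.
; Stopping also removes the focus prompt widget on cosmetic-capable machines, and runs when the ability ends.
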(event Ability|EventActivateAbility
  (bind _focusdata_forcefocusenabled_6_8b220e3a4c0aafa126bd4286f5599f22 (Utilities|Struct|BreakEventsSharedForceFocusEvent (Variables|GetFocusData_ForceFocusEnabled_6_8B220E3A4C0AAFA126BD4286F5599F22)))
  ; Begin Ability / Listen for Focus Event
  (Class|FortAsyncActionCurrentPlaylistReady|PlaylistDataReadyAsync
    (:OnReady
      (bind (_asynctaskproxy _payload) (GameplayEventRouter|StartListeningToStatefulEvent (GameplayEventRouter|GetGlobalEventRouter) "/Events_Shared/Gameplay/ForceFocus/EventsShared_ForceFocus_Event.EventsShared_ForceFocus_Event")
        (:OnEventReceived
          (Variables|Default|SetFocusData _payload)
          (CallFunction|PCShippingLog "(GA_EventsShared_ForceFocus - On Event Received)")
          (if _focusdata_forcefocusenabled_6_8b220e3a4c0aafa126bd4286f5599f22
            (CallFunction|StartFocusing)
            (else
              (CallFunction|StopFocusing))))))))

(event Ability|EventOnEndAbility
  (params (bWasCancelled :type Boolean))
  ; End Ability
  (CallFunction|PCShippingLog "(End Ability)")
  (CallFunction|StopFocusing) ; Make sure the focus turns off if the ability\nis being cancelled.
)

(event Custom|Stop%20Focusing
  (bind _returnvalue (Class|SpecialEventGameplayLibrary|GetSpecialEventMutator))
  (bind _returnvalue_2 (Fortnite|Athena|GameState|GetGameStateAthena))
  (bind (_focusdata_forcefocusenabled_6_8b220e3a4c0aafa126bd4286f5599f22 _focusdata_duration_10_3674387b40d646512d5dca9727ec0275 _focusdata_focalpointfov_13_8a2cbf5c4c7d57cb2d426dbddf08bf60 _focusdata_interpspeed_15_4a87062643da51a3e2cf258698fd5e2a) (Utilities|Struct|BreakEventsSharedForceFocusEvent (Variables|GetFocusData_ForceFocusEnabled_6_8B220E3A4C0AAFA126BD4286F5599F22)))
  ; Stop Focusing
  (Networking|CanExecuteCosmeticEvents
    (:True
      (CallFunction|ToggleWidget false))
    (:False
      (Utilities|IsValid _returnvalue_2
        (:Is%20Valid
          (Class|FortGameStateAthena|SetFocalPointActor _returnvalue_2 0 "0, 0, 0" 9e+01 0.0 _focusdata_interpspeed_15_4a87062643da51a3e2cf258698fd5e2a)
          (CallFunction|PCShippingLog "(Stop Focusing)")
          (Utilities|IsValid _returnvalue
            (:Is%20Valid
              (Event|SetLockingonFocalPoint _returnvalue false)
              (Widget|SetVisibility (Variables|Default|GetFocusPromptWidget) "Hidden"))))))))

(event Custom|Start%20Focusing
  (bind _returnvalue (Class|SpecialEventGameplayLibrary|GetSpecialEventMutator))
  (bind (_focusdata_forcefocusenabled_6_8b220e3a4c0aafa126bd4286f5599f22 _focusdata_duration_10_3674387b40d646512d5dca9727ec0275 _focusdata_focalpointfov_13_8a2cbf5c4c7d57cb2d426dbddf08bf60 _focusdata_interpspeed_15_4a87062643da51a3e2cf258698fd5e2a _focusdata_focalpointoffset_18_117acaef4a558f4cae3f0b953a32672c _focusdata_focusactor_9_1763734a4f76a6eb17b0f985ce9a0797) (Utilities|Struct|BreakEventsSharedForceFocusEvent (Variables|GetFocusData_ForceFocusEnabled_6_8B220E3A4C0AAFA126BD4286F5599F22)))
  ; Start Focusing
  (Networking|CanExecuteCosmeticEvents
    (:True)
    (:False
      (Utilities|IsValid _focusdata_focusactor_9_1763734a4f76a6eb17b0f985ce9a0797
        (:Is%20Valid
          (Class|FortGameStateAthena|SetFocalPointActor (Fortnite|Athena|GameState|GetGameStateAthena) _focusdata_focusactor_9_1763734a4f76a6eb17b0f985ce9a0797 _focusdata_focalpointoffset_18_117acaef4a558f4cae3f0b953a32672c _focusdata_focalpointfov_13_8a2cbf5c4c7d57cb2d426dbddf08bf60 _focusdata_duration_10_3674387b40d646512d5dca9727ec0275 _focusdata_interpspeed_15_4a87062643da51a3e2cf258698fd5e2a)
          (CallFunction|PCShippingLog "(Set Focal Point Actor)")
          (Utilities|IsValid _returnvalue
            (:Is%20Valid
              (Event|SetLockingonFocalPoint _returnvalue true))))))))

(event Custom|ToggleWidget
  (params (WidgetActive :type Boolean))
  (bind _focuspromptwidget (Variables|Default|GetFocusPromptWidget))
  ; Press Button to Focus Widget
  (CallFunction|ToggleMobileFocus WidgetActive)
  (if WidgetActive
    (Utilities|IsValid (Variables|Default|GetFocusPromptWidget)
      (:Is%20Not%20Valid
        (bind _returnvalue (UserInterface|CreateWidget "/SpecialEventGameplay/UI/SpecialEvent_FocusPrompt.SpecialEvent_FocusPrompt_C"))
        (bind _output_get (Variables|Default|SetFocusPromptWidget _returnvalue))
        (UserInterface|Viewport|AddtoViewport _output_get 0)
        (Widget|SetVisibility _returnvalue "Visible")))
    (else
      (Utilities|IsValid _focuspromptwidget
        (:Is%20Valid
          (Widget|RemovefromParent _focuspromptwidget)
          (Variables|Default|SetFocusPromptWidget 0))))))

(event Custom|PC%20Shipping%20Log
  (params (string :type String))
  (bind (_focusdata_forcefocusenabled_6_8b220e3a4c0aafa126bd4286f5599f22 _focusdata_duration_10_3674387b40d646512d5dca9727ec0275 _focusdata_focalpointfov_13_8a2cbf5c4c7d57cb2d426dbddf08bf60 _focusdata_interpspeed_15_4a87062643da51a3e2cf258698fd5e2a _focusdata_focalpointoffset_18_117acaef4a558f4cae3f0b953a32672c _focusdata_focusactor_9_1763734a4f76a6eb17b0f985ce9a0797) (Utilities|Struct|BreakEventsSharedForceFocusEvent (Variables|GetFocusData_ForceFocusEnabled_6_8B220E3A4C0AAFA126BD4286F5599F22)))
  ; Shipping Log
  (Fortnite|FortShippingLog 0 (Utilities|String|Append (Utilities|String|Append "[Pistol Canary] **GA_PistolCanary_Focus_Forced** " string) (Utilities|String|ToString(Boolean) _focusdata_forcefocusenabled_6_8b220e3a4c0aafa126bd4286f5599f22) ", Duration: " (Utilities|String|ToString(Float) _focusdata_duration_10_3674387b40d646512d5dca9727ec0275) ", FOV: " (Utilities|String|ToString(Float) _focusdata_focalpointfov_13_8a2cbf5c4c7d57cb2d426dbddf08bf60) ", Interp Speed: " (Utilities|String|ToString(Float) _focusdata_interpspeed_15_4a87062643da51a3e2cf258698fd5e2a) ", actor: " (Utilities|GetDisplayName _focusdata_focusactor_9_1763734a4f76a6eb17b0f985ce9a0797) (Utilities|String|SelectString ", (server)" ", (client)" (Networking|IsServer)))))

(event Custom|ToggleMobileFocus
  (params (ShouldShow :type Boolean))
  (bind _focuspromptwidget (Variables|Default|GetFocusPromptWidget))
  ; Press Button to Focus Widget
  (Widget|SetVisibility _focuspromptwidget "Visible")
  (Variables|/Script/UMG.Widget|SetVisibility "Visible" _focuspromptwidget))
